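# Physics asset for the Dark Minion male one-handed pickaxe mesh, giving its hinged arm and feather bones simulated collision bodies.
# Capsules cover the handle, the arm hinge and three short feather chains, with the hinge loosely limited against the handle.
# Feather chain A hangs from the handle, while chains B and C hang from the hinge.
# Tight swing and twist limits narrow further at each chain tip, keeping feather motion subtle.
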PhysicsAsset /BRCosmetics/Weapons/FORT_Melee/Pickaxe_Dark_Minion_Male_1H/Meshes/Dark_Minion_Male_1h_Axe_Physics
bodies {
 Arm_hing_02 [{shapeType=Capsule, length=23.5682, radius=11.4184, center=<-22.9088 0 20.59>, rotation=<rot 0 90 -88.2875>}]
 Melee_TwoHanded_Handle [{shapeType=Capsule, length=19.8372, radius=14.3731, center=<0.001 3.2261 13.205>, rotation=<rot -26.5074 -90 -0.0377>}]
 dyn_feather_A_01 [{shapeType=Capsule, length=3.6656, radius=2, center=<4.1612 0 0>, rotation=<rot 90 0 0>}]
 dyn_feather_A_02 [{shapeType=Capsule, length=3.6656, radius=2, center=<4.1612 0 0>, rotation=<rot 90 0 0>}]
 dyn_feather_A_03 [{shapeType=Capsule, length=3.6656, radius=2, center=<4.1612 0 0>, rotation=<rot 90 0 0>}]
 dyn_feather_B_01 [{shapeType=Capsule, length=3.6656, radius=2, center=<4.1612 0 0>, rotation=<rot 90 0 0>}]
 dyn_feather_B_02 [{shapeType=Capsule, length=3.6656, radius=2, center=<4.1612 0 0>, rotation=<rot 90 0 0>}]
 dyn_feather_B_03 [{shapeType=Capsule, length=3.6656, radius=2, center=<4.1612 0 0>, rotation=<rot 90 0 0>}]
 dyn_feather_B_04 [{shapeType=Capsule, length=3.6656, radius=2, center=<4.1612 0 0>, rotation=<rot 90 0 0>}]
 dyn_feather_C_01 [{shapeType=Capsule, length=3.6656, radius=2, center=<4.1612 0 0>, rotation=<rot 90 0 0>}]
 dyn_feather_C_02 [{shapeType=Capsule, length=3.6656, radius=2, center=<4.1612 0 0>, rotation=<rot 90 0 0>}]
 dyn_feather_C_03 [{shapeType=Capsule, length=3.6656, radius=2, center=<4.0589 -0.7722 0.9067>, rotation=<rot 90 0 9.5845>}]
}
constraints [
 {bone1Name=Arm_hing_02, bone2Name=Melee_TwoHanded_Handle, swing1LimitDegrees=45, swing2LimitDegrees=45, twistLimitDegrees=45}
 {bone1Name=dyn_feather_C_01, bone2Name=Arm_hing_02, swing1LimitDegrees=10, swing2LimitDegrees=10, twistLimitDegrees=3}
 {bone1Name=dyn_feather_C_02, bone2Name=dyn_feather_C_01, swing1LimitDegrees=10, swing2LimitDegrees=10, twistLimitDegrees=3}
 {bone1Name=dyn_feather_C_03, bone2Name=dyn_feather_C_02, swing1LimitDegrees=5, swing2LimitDegrees=5, twistLimitDegrees=2}
 {bone1Name=dyn_feather_B_01, bone2Name=Arm_hing_02, swing1LimitDegrees=10, swing2LimitDegrees=10, twistLimitDegrees=3}
 {bone1Name=dyn_feather_B_02, bone2Name=dyn_feather_B_01, swing1LimitDegrees=10, swing2LimitDegrees=10, twistLimitDegrees=3}
 {bone1Name=dyn_feather_B_03, bone2Name=dyn_feather_B_02, swing1LimitDegrees=10, swing2LimitDegrees=10, twistLimitDegrees=3}
 {bone1Name=dyn_feather_B_04, bone2Name=dyn_feather_B_03, swing1LimitDegrees=5, swing2LimitDegrees=5, twistLimitDegrees=2}
 {bone1Name=dyn_feather_A_01, bone2Name=Melee_TwoHanded_Handle, swing1LimitDegrees=10, swing2LimitDegrees=10, twistLimitDegrees=3}
 {bone1Name=dyn_feather_A_02, bone2Name=dyn_feather_A_01, swing1LimitDegrees=10, swing2LimitDegrees=10, twistLimitDegrees=3}
 {bone1Name=dyn_feather_A_03, bone2Name=dyn_feather_A_02, swing1LimitDegrees=5, swing2LimitDegrees=5, twistLimitDegrees=2}
]
