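# Animation Blueprint for the Dark Minion Male one-handed pickaxe that switches its arm between retracted and extended poses.
# AnimGraph blends hinge and slide bone overrides by the retracted flag, then applies a slot and RigidBody physics, weaker in the front end.
# EventGraph polls owner montages: equip extends it after a short delay, melee extends it immediately.
# Weapon changes retract it, as does a delay after harvest swings end.
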AnimBlueprint /BRCosmetics/Weapons/FORT_Melee/Pickaxe_Dark_Minion_Male_1H/Meshes/Dark_Minion_Male_1h_Axe_Skeleton_AnimBlueprint
target_skeleton=&/Dark_Minion_Male_1h_Axe_Skeleton
graphs [
 AnimGraph {nodes [
   n1 = "Transform (Modify) Bone - Bone: Arm_hing_01"(ComponentPose=Local To Component(), Alpha=1, Translation="0,0,0", Rotation="-179,0,0", Scale="1,1,1") {rotationMode=BMM_Replace, alphaBoolBlend {blendOption=Linear}, boneToModify {boneName=Arm_hing_01}, rotation {pitch=-179}}
   n2 = "Transform (Modify) Bone - Bone: Arm_slide"(ComponentPose=@n1, Alpha=1, Translation="13,0,0", Rotation="0,0,0", Scale="1,1,1") {
    translationMode=BMM_Replace
    translationSpace=BCS_BoneSpace
    alphaBoolBlend {blendOption=Linear}
    boneToModify {boneName=Arm_slide}
    translation {x=13}
   }
   n3 = Blend Poses by bool(
    BlendPose_0=Component To Local(ComponentPose=@n2)
    BlendPose_1=Component To Local(ComponentPose=Local To Component())
    bActiveValue=Get retracted() -> (retracted:bool)
   ) {blendType=Linear, childUpateMode=ResetChildOnActivate}
   n4 = RigidBody!(
    ComponentPose=Local To Component(LocalPose=Slot 'DefaultSlot'(Source=@n3))
    Alpha=SelectFloat(A=0.35, B=1, bPickA=Get bIsInFrontEnd())
    OverrideWorldGravity:Vector="0,0,-100"
    ExternalForce
   ) -> (Pose) {
    bClampLinearTranslationLimitToRefPose=true
    bOverrideWorldGravity=true
    overridePhysicsAsset=&/Dark_Minion_Male_1h_Axe_Physics
    simulationSpace=BaseBoneSpace
    alphaBoolBlend {blendOption=Linear}
    baseBoneRef {boneName=Melee_TwoHanded_Handle}
    componentAppliedLinearAccClamp=<5000 5000 5000>
    componentLinearAccScale=<0.4 0.4 0.4>
    componentLinearVelScale=<0.4 0.4 0.4>
    overrideWorldGravity {z=-380}
   }
   Output Pose(Result=Component To Local(ComponentPose=@n4)) {layerGroup=None}
  ]}
 EventGraph {nodes [
   Event OnWeaponChanged()
   > Set retracted(retracted:bool=true) -> (Output_Get:bool)
   Event BlueprintInitializeAnimation()
   > Set Skel_Mesh(Skel_Mesh=GetOwnerSkeletalMeshComponent())
   > n1 = Cast To FortPlayerPawn(Object=TryGetFortPawn())
   > Set Pawn(Pawn:Fort Player Pawn Object Reference=@n1) -> (Output_Get:Fort Player Pawn Object Reference)
   Event BlueprintUpdateAnimation()
   > n2 = Sequence!(execute) -> (then_0, then_1, then_2:Exec, then_3:Exec, then_4:Exec)
   > Set IsFrontEndAnimInstance(IsFrontEndAnimInstance:bool=IsValid(Object=GetOwnerFrontEndAnimInstance())) -> (Output_Get:bool)
   n3 = Branch(execute=@n2.then_1, Condition=Get IsFrontEndAnimInstance() -> (IsFrontEndAnimInstance:bool))
   > n4 = Branch(Condition=Montage_IsPlaying(
     self=GetOwnerAnimInstance!(self:Custom Character Part Anim Instance Object Reference) -> (ReturnValue:Anim Instance Object Reference)
     Montage=/Game/Animation/Game/MainPlayer/Combat/Melee/Galileo_Ferry/Galileo_Ferry_Equip_M
    ))
   Delay(execute=[@n4.then, @n5.then], Duration=0.33)
   > Set retracted(retracted=false)
   > Set swing(swing:bool=false) -> (Output_Get:bool)
   > Set Was Swinging(Was Swinging:bool=false) -> (Output_Get:bool)
   n5 = Branch(
    execute=@n3.else
    Condition=Montage_IsPlaying(
     self=GetOwnerFortPlayerAnimInstance!(self:Custom Character Part Anim Instance Object Reference) -> (ReturnValue:Fort Player Anim Instance Object Reference)
     Montage=&Galileo_Ferry_Equip_M
    )
   )
   n6 = Branch(execute=@n2.then_2, Condition=Get IsFrontEndAnimInstance())
   n7 = Montage_IsPlaying(self=GetOwnerAnimInstance(), Montage=/Game/Animation/Game/MainPlayer/Combat/Melee/Galileo_Ferry/Galileo_Ferry_Harvesting_Combo_Montage)
   > Set Is_Swinging(Is_Swinging:double="To Float (Boolean)"(InBool=@n7)) -> (Output_Get:double)
   > Set Was Swinging(Was Swinging=Get swing() -> (swing:bool))
   > Set swing(swing=@n7)
   n8 = Montage_IsPlaying(self=GetOwnerAnimInstance(), Montage=&Galileo_Ferry_Harvesting_Combo_Montage)
   Set Is_Swinging(execute=@n6.else, Is_Swinging="To Float (Boolean)"(InBool=@n8))
   > Set Was Swinging(Was Swinging=Get swing())
   > Set swing(swing=@n8)
   n9 = Get swing()
   Branch(execute=@n2.then_3, Condition=AND Boolean(A=Get Was Swinging() -> (Was Swinging:bool), B=NOT Boolean(A=@n9)))
   > Delay(Duration=2)
   > Set retracted(retracted=true)
   Branch(execute=@n2.then_4, Condition=Get bIsPlayingMeleeAnim())
   > Set retracted(retracted=false)
   TryGetPawnOwner()
  ]}
]
variables [
 Is_Swinging {type=number}
 IsFrontEndAnimInstance {type=boolean}
 Pawn {type=FortPlayerPawn}
 retracted {type=boolean}
 Skel_Mesh {type=SkeletalMeshComponent}
 Stage1 {type=number}
 swing {type=boolean}
 Was Swinging {type=boolean}
]
